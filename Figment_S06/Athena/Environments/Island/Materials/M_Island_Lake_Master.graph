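# Master landscape-style material for lake island terrain, layering rock, mud, lake grass, dry dirt and flowers through material attributes.
# Two-scale rock with optional island mask and normal textures height-blends by slope into mud and grass.
# Height-aware vertex-color blends place grass (A), top mud (G), dry dirt (B) and flowers (R), using mostly world-XY tiling.
# Shading path switches simplify Forward and Mobile, whose final attributes keep only base color with fixed specular and roughness.
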Material /Figment_S06/Athena/Environments/Island/Materials/M_Island_Lake_Master
graph {
 properties {bUseMaterialAttributes=true, pixelDepthOffsetMode=PDOM_Legacy, usedShadingModels=Default Lit}
 nodes [
  Clamp(None=LinearInterpolate(
    A=ScalarParameter() {defaultValue=-0.3596, parameterName=Transition_M1}
    B=ScalarParameter() {defaultValue=1.3865, parameterName=Transition_M2}
    Alpha=MaterialFunctionCall() {materialFunction=&LinearGradient}.VGradient
   ))
  n1 = StaticBoolParameter() {defaultValue=true, parameterName=Rock_UseWorldAlign}
  n2 = ScalarParameter() {defaultValue=2000, group=Rocks, parameterName=ScaleRocks}
  n3 = MaterialFunctionCall!(UseWorldAlign=@n1, TextureScale=@n2) {materialFunction=&/MF_IslandRocks_Lake}
  n4 = MaterialFunctionCall(UseWorldAlign=@n1, TextureScale=Divide(A=@n2)) {materialFunction=&MF_IslandRocks_Lake}
  n5 = MaterialFunctionCall(
   A=Multiply(A=ScalarParameter() {defaultValue=0.5, group=Rocks, parameterName=BaseBodyBrightness}, B=@n3.RGB)
   B=Multiply(A=ScalarParameter() {defaultValue=2, group=Rocks, parameterName=EdgeBrightness}, B=@n4.RGB)
   Transition Phase=Constant() {r=0.33}
   Height Texture=@n3.A
  ) {materialFunction=&HeightLerp}
  ComponentMask(None=@n5.Results) {g=true, r=true}
  n6 = MaterialFunctionCall!(Height Offset Amount, Texture mip bias amount, Ground texture mask) {materialFunction=/Game/WaterAndWind/Materials/Water/ShallowWaterFunction}
  MaterialFunctionCall!(Base Material, Top Material=@n6.MA Result, Alpha) {materialFunction=/Engine/Functions/MaterialLayerFunctions/MatLayerBlend_Standard}
  Reroute(None=@n6.Normal)
  n7 = Divide(A=ComponentMask(None=WorldPosition().XYZ) {g=true, r=true}, B=ScalarParameter() {defaultValue=25, parameterName=MudTextureTiling})
  n8 = TextureSample(UVs=@n7) {samplerSource=SSM_Wrap_WorldGroupSettings, texture=/Game/Packages/DS_Fortnight/WhiteRockCustom_D}
  Saturate(None=LinearInterpolate(A=Constant() {r=-1}, B=Constant() {r=1}, Alpha=@n8.A))
  n9 = VectorParameter() {group=Rocks, parameterName=ColorIsland, defaultValue=<rgba 0.2344 0.1401 0.0732 0>}
  n10 = Multiply(A=@n9.RGB, B=@n5.Results)
  n11 = Constant() {r=0.5}
  n12 = TextureSampleParameter2D() {
   parameterName=M
   samplerSource=SSM_Wrap_WorldGroupSettings
   samplerType=SAMPLERTYPE_Masks
   texture=/Game/Athena/Environments/Island/Textures/T_Island2_M
  }
  n13 = Constant() {r=0.3}
  n14 = StaticSwitchParameter(
   True=LinearInterpolate(
    A=VectorParameter() {group=Rocks, parameterName=ColorIsland_AO, defaultValue=<rgba 0.1667 0.0515 0.0035 0>}.RGB
    B=MaterialFunctionCall(A=@n10, B=@n9.RGB, Transition Phase=@n11, Height Texture=@n12.R, Contrast=@n13) {materialFunction=&HeightLerp}.Results
    Alpha=@n12.G
   )
   False=@n10
  ) {defaultValue=true, parameterName=UseIslandAssets}
  n15 = StaticSwitchParameter(True=LinearInterpolate(A=Constant() {r=0.1}, B=Constant() {r=0.5}, Alpha=@n12.G), False=Constant()) {defaultValue=true, parameterName=UseIslandAssets}
  n16 = Constant() {r=1}
  n17 = Constant() {r=0.55}
  n18 = TextureSampleParameter2D() {
   parameterName=N
   samplerSource=SSM_Wrap_WorldGroupSettings
   samplerType=SAMPLERTYPE_Normal
   texture=/Game/Athena/Environments/Island/Textures/T_Island2_N
  }
  n19 = StaticSwitchParameter(
   True=@n3.N
   False=LinearInterpolate(
    A=@n3.N
    B=MaterialFunctionCall(BaseNormal=@n3.N, AdditionalNormal=@n4.N) {materialFunction=&BlendAngleCorrectedNormals}.Result
    Alpha=@n5.Alpha
   )
  ) {defaultValue=true, parameterName=UseIslandAssets}
  n20 = TextureSampleParameter2D() {
   parameterName=N
   samplerSource=SSM_Wrap_WorldGroupSettings
   samplerType=SAMPLERTYPE_Normal
   texture=&T_Island2_N
  }
  n21 = LinearInterpolate(
   A=LinearInterpolate(A=Constant3Vector() {constant=<rgba 0 0 1 0>}, B=@n19)
   B=@n19
   Alpha=MaterialFunctionCall!(up/Down Height Mask, N=@n20.RGB) {materialFunction=/Game/Athena/Environments/Island/Materials/MF_IslandWM_Game}.MaskOut Bottom Details
  )
  n22 = MaterialFunctionCall(
   Normal=ShadingPathSwitch(Default=@n21, Deferred=@n21, Forward=@n19, Mobile=@n19)
   Flatness=ScalarParameter() {defaultValue=0.5, parameterName=Rock_NormalFlatten}
  ) {materialFunction=&FlattenNormal}
  n23 = MakeMaterialAttributes!(
   BaseColor=@n14
   Metallic
   Specular=@n15
   Roughness=StaticSwitchParameter(True=LinearInterpolate(A=@n16, B=@n17, Alpha=@n12.G), False=Constant() {r=0.7}) {defaultValue=true, parameterName=UseIslandAssets}
   Anisotropy
   EmissiveColor
   Opacity
   OpacityMask
   Normal=StaticSwitchParameter(
    True=MaterialFunctionCall(BaseNormal=@n18.RGB, AdditionalNormal=@n22.Result) {materialFunction=&BlendAngleCorrectedNormals}.Result
    False=@n22.Result
   ) {defaultValue=true, parameterName=UseIslandAssets}
   Tangent
   WorldPositionOffset
   SubsurfaceColor
   ClearCoat
   ClearCoatRoughness
   AmbientOcclusion
   Refraction
   CustomizedUVs_0
   CustomizedUVs_1
   CustomizedUVs_2
   CustomizedUVs_3
   CustomizedUVs_4
   CustomizedUVs_5
   CustomizedUVs_6
   CustomizedUVs_7
   PixelDepthOffset
   ShadingModel
   Displacement
  )
  n24 = SetMaterialAttributes!(MaterialAttributes, Base Color=@n14, Specular=@n15) {attributeSetTypes [#1, #2]}
  n25 = ShadingPathSwitch(Default=@n23, Deferred=@n23, Forward=@n24, Mobile=@n24)
  n26 = VertexNormalWS()
  n27 = Constant3Vector() {constant=<rgba 0 0 1 0>}
  n28 = Multiply(A=VectorParameter() {group=Mud, parameterName=MudTintDry, defaultValue=<rgba 0.3281 0.2739 0.2358 0>}.RGB, B=@n8.RGB)
  n29 = TextureSample(UVs=@n7) {samplerSource=SSM_Wrap_WorldGroupSettings, samplerType=SAMPLERTYPE_Normal, texture=/Game/Packages/DS_Fortnight/WhiteRockCustom_N}
  n30 = SetMaterialAttributes!(
   MaterialAttributes
   Base Color=@n28
   Specular=LinearInterpolate(
    A=ScalarParameter() {defaultValue=0.05, parameterName=Dirt_SpecMin}
    B=ScalarParameter() {defaultValue=0.55, parameterName=Dirt_Spec_Max}
    Alpha=@n8.A
   ) {constB=0.4}
   Roughness=LinearInterpolate(
    A=ScalarParameter() {defaultValue=0.99, parameterName=Dirt_RMin}
    B=ScalarParameter() {defaultValue=0.65, parameterName=Dirt_RMax}
    Alpha=@n8.A
   )
   Normal=@n29.RGB
   Opacity=@n8.A
  ) {attributeSetTypes [
    #1
    #2
    #3
    #4
    #5
   ]}
  n31 = ComponentMask(None=@n8.RGB) {r=true}
  n32 = ShadingPathSwitch(
   Default=@n30
   Deferred=@n30
   Forward=SetMaterialAttributes!(MaterialAttributes, Base Color=@n28, Normal=@n29.RGB, Specular=@n31) {attributeSetTypes [#1, #4, #2]}
   Mobile=SetMaterialAttributes!(MaterialAttributes, Base Color=@n28, Specular=@n31) {attributeSetTypes [#1, #2]}
  )
  n33 = MaterialFunctionCall!(Water, GrassColor) {materialFunction=/Game/Athena/Environments/Landscape/MaterialFunctions/Standard/MF_Athena_LakeGrass_01}
  n34 = ScalarParameter() {defaultValue=0.5, group=Mud, parameterName=GrassMudShadowTintA}
  n35 = ScalarParameter() {defaultValue=2, group=Mud, parameterName=GrassMudShadowPowerA}
  n36 = MaterialFunctionCall!(Mat attributes=@n33.Result) {materialFunction=/Engine/Functions/Engine_MaterialFunctions02/MatLayerBlend_BreakOpacity}
  n37 = OneMinus(None=MaterialFunctionCall(
    Transition Phase=VertexColor().A
    Height Texture=OneMinus(None=@n36.Result)
    Contrast=ScalarParameter() {defaultValue=1, parameterName=Grass_BlendContrast}
   ) {materialFunction=&HeightLerp}.Alpha)
  n38 = Clamp(None=Multiply(A=Power(Base=@n37, Exp=@n35), B=ScalarParameter() {defaultValue=1.5, group=Mud, parameterName=GrassMudShadowXA})) {clampMode=CMODE_ClampMax}
  n39 = OneMinus(None=@n38)
  n40 = BlendMaterialAttributes(
   A=@n32
   B=MaterialFunctionCall!(
    Base Material=MaterialFunctionCall!(Material=@n33.Result, Tint=@n34, TintMask=@n39) {materialFunction=/Engine/Functions/MaterialLayerFunctions/MatLayerBlend_Tint}.Blended Material
    SpecularMultiplier=@n38
   ) {materialFunction=/Engine/Functions/MaterialLayerFunctions/MatLayerBlend_ModulateSpecular}.Blended Material
   Alpha=@n37
  )
  n41 = MaterialFunctionCall(
   Transition Phase=LinearInterpolate(
    A=ScalarParameter() {defaultValue=-0.5, parameterName=Rock_BlendMin}
    B=ScalarParameter() {defaultValue=1.15, parameterName=Rock_BlendMax}
    Alpha=DotProduct(A=@n26, B=@n27)
   ) {constA=-1}
   Height Texture=MaterialFunctionCall(Mat attributes=@n40) {materialFunction=&MatLayerBlend_BreakOpacity}.Result
   Contrast=ScalarParameter() {defaultValue=5, parameterName=Rock_BlendContrast}
  ) {materialFunction=&HeightLerp}
  n42 = BlendMaterialAttributes(
   A=StaticSwitchParameter(True=BlendMaterialAttributes(A=@n25, B=@n40, Alpha=@n41.Alpha), False=@n40) {defaultValue=true, parameterName=Use_Rock}
   B=@n32
   Alpha=MaterialFunctionCall(
    Transition Phase=VertexColor().G
    Height Texture=MaterialFunctionCall(Mat attributes=@n32) {materialFunction=&MatLayerBlend_BreakOpacity}.Result
    Contrast=ScalarParameter() {parameterName=TopMud_BlendContrast}
   ) {materialFunction=&HeightLerp}.Alpha
  )
  n43 = Divide(A=ComponentMask(None=WorldPosition().XYZ) {g=true, r=true}, B=ScalarParameter() {defaultValue=1024, parameterName=DryDirt_Scale})
  n44 = TextureSample(UVs=@n43) {texture=/Game/Athena/Environments/Island/Textures/Dirt_D}
  n45 = SetMaterialAttributes!(
   MaterialAttributes
   Base Color=Multiply(A=@n44.RGB, B=Constant3Vector() {constant=<rgba 0.4062 0.3904 0.347 1>})
   Specular=Constant() {r=0.04}
   Roughness=LinearInterpolate(A=Constant() {r=0.4}, B=Constant() {r=0.8}, Alpha=@n44.A)
   Normal=MaterialFunctionCall(
    Normal=TextureSample(UVs=@n43) {samplerType=SAMPLERTYPE_Normal, texture=/Game/Packages/DS_Fortnight/ExperimentalTextures/Dirt_N}.RGB
    Flatness=ScalarParameter() {defaultValue=0.5, parameterName=DryDirt_NormalFlatten}
   ) {materialFunction=&FlattenNormal}.Result
  ) {attributeSetTypes [
    #1
    #2
    #3
    #4
   ]}
  n46 = VertexColor()
  n47 = OneMinus(None=MaterialFunctionCall(Mat attributes=@n42) {materialFunction=&MatLayerBlend_BreakOpacity}.Result)
  n48 = ScalarParameter() {defaultValue=0.25, parameterName=BlueChannel_BlendContrast}
  n49 = OneMinus(None=@n41.Alpha)
  n50 = Clamp(None=Subtract(
    A=MaterialFunctionCall(Transition Phase=@n46.B, Height Texture=@n47, Contrast=@n48) {materialFunction=&HeightLerp}.Alpha
    B=@n49
   ))
  n51 = Clamp(None=Subtract(A=@n46.B, B=@n49))
  n52 = BlendMaterialAttributes(A=@n42, B=@n45, Alpha=ShadingPathSwitch(Default=@n50, Deferred=@n50, Forward=@n51, Mobile=@n51))
  n53 = Divide(A=ComponentMask(None=WorldPosition().XYZ) {g=true, r=true}, B=ScalarParameter() {defaultValue=700, parameterName=Flower_TexScale_1})
  n54 = TextureSample(UVs=@n53) {texture=/Figment_S06/Athena/Environments/Landscape/Textures/T_Athena_Flower_Ground_D}
  n55 = TextureSample(UVs=@n53) {samplerType=SAMPLERTYPE_LinearColor, texture=/Figment_S06/Athena/Environments/Landscape/Textures/T_Athena_Flower_Ground_M}
  n56 = LinearInterpolate(
   A=MaterialFunctionCall!(Material=@n52) {materialFunction=/Engine/Functions/MaterialLayerFunctions/MatLayerBlend_BreakBaseColor}.BaseColor
   B=@n54.RGB
   Alpha=@n55.G
  )
  n57 = LinearInterpolate(A=ScalarParameter() {parameterName=Flowers_RMin}, B=ScalarParameter() {defaultValue=1, parameterName=Flowers_RMax}, Alpha=@n55.R)
  n58 = VertexColor()
  n59 = MaterialFunctionCall(Transition Phase=@n58.R, Height Texture=@n55.G, Contrast=ScalarParameter() {parameterName=Flower_BlendContrast}) {materialFunction=&HeightLerp}
  n60 = GetMaterialAttributes(None=BlendMaterialAttributes(
    A=@n52
    B=SetMaterialAttributes!(MaterialAttributes, Base Color=@n56, Specular=@n54.R, Roughness=@n57) {attributeSetTypes [#1, #2, #3]}
    Alpha=ShadingPathSwitch(Default=@n59.Alpha, Deferred=@n59.Alpha, Forward=@n58.R, Mobile=@n58.R)
   )) {attributeGetTypes [#1]}
  n61 = Multiply(A=@n60.Base Color, B=@n55.B)
  n62 = ScalarParameter() {defaultValue=0.5, parameterName=FlowerAOAmount}
  n63 = Multiply(A=@n59.Alpha, B=ScalarParameter() {defaultValue=3.0392, parameterName=fdghdfhf})
  n64 = SetMaterialAttributes!(
   MaterialAttributes=@n60.MaterialAttributes
   Base Color=LinearInterpolate(A=@n60.Base Color, B=LinearInterpolate(A=@n56, B=@n61, Alpha=@n62), Alpha=Clamp(None=@n63))
  ) {attributeSetTypes [#1]}
  n65 = StaticSwitchParameter(True=@n64, False=@n64) {defaultValue=true, desc=Legacy requirement for grass generation, parameterName="Use Grass(true) or a BaseTilingTexture(False)"}
  n66 = MakeMaterialAttributes(
   BaseColor=GetMaterialAttributes(None=@n65) {attributeGetTypes [#1]}.Base Color
   Specular=Constant()
   Roughness=Constant() {r=1}
  )
  ShadingPathSwitch(Default=@n65, Deferred=@n65, Forward=@n66, Mobile=@n66)
  n67 = TextureSampleParameter2D!(UVs=Panner(Coordinate=TextureCoordinate() {uTiling=0.5, vTiling=0.25}) {speedY=0.05}, Bias, Apply View MipBias) {
   mipValueMode=TMVM_MipBias
   parameterName=Normal_1
   samplerType=SAMPLERTYPE_Normal
   texture=/Figment_S06/Packages/Fortnite_UniqueMaterials_1/Textures/T_Water2_N
  }
  ShadingPathSwitch(Default=LinearInterpolate(A=@n67.RGB, B=@n6.Normal), Mobile=@n6.Normal) {desc=Migrated from FeatureLevelSwitch}
 ]
}
