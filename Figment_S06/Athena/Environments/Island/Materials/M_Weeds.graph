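# M_Weeds is a masked, two-sided foliage material for Athena island weeds, usable on instanced static meshes.
# World XY projection maps tinted island-green color and a flattened normal, while color alpha times a radial gradient drives a lerp.
# Panning swirl noise at a separate wind scale perturbs the normal and is also amplified in XY, suggesting wind sway.
# A world-projected blend mask drives another lerp, replaced by a constant on the mobile shading path.
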Material /Figment_S06/Athena/Environments/Island/Materials/M_Weeds
graph {
 properties {
  bUsedWithInstancedStaticMeshes=true
  blendMode=BLEND_Masked
  pixelDepthOffsetMode=PDOM_Legacy
  shadingModel=MSM_TwoSidedFoliage
  usedShadingModels=Two Sided Foliage
 }
 nodes [
  n1 = Divide(A=ComponentMask(None=WorldPosition().XYZ) {g=true, r=true}, B=ScalarParameter() {defaultValue=447.5602, parameterName=Scale})
  n2 = Divide(A=ComponentMask(None=WorldPosition().XYZ) {g=true, r=true}, B=ScalarParameter() {defaultValue=1459.5212, parameterName=ScaleWind})
  n3 = Multiply(
   A=TextureSample(UVs=Panner(Coordinate=@n2) {speedX=0.1, speedY=0.1}) {samplerType=SAMPLERTYPE_Normal, texture=/Game/Global/Textures/Default/Noise/T_SwirlNoise_N}.RGB
   B=Constant3Vector() {constant=<rgba 0.5 0.5 0 1>}
  )
  Add(
   A=MaterialFunctionCall(
    Normal=TextureSample(UVs=@n1) {samplerType=SAMPLERTYPE_Normal, texture=/Game/Athena/Environments/Island/Textures/T_IslandGreen_N}.RGB
    Flatness=Constant() {r=0.3}
   ) {materialFunction=&FlattenNormal}.Result
   B=@n3
  )
  Constant3Vector() {constant=<rgba 0.1437 0.151 0.0117 0>}
  Constant() {r=0.15}
  Constant() {r=0.6}
  n4 = TextureSample(UVs=@n1) {texture=/Game/Athena/Environments/Island/Textures/T_IslandGreen_D}
  n5 = Constant() {r=0.5}
  n6 = Constant() {r=3}
  LinearInterpolate(
   A=Constant() {r=0.2}
   B=Constant() {r=2}
   Alpha=Multiply(
    A=@n4.A
    B=MaterialFunctionCall(Radius=@n5, Density=@n6) {materialFunction=&RadialGradientExponential}.RadialGradientExponential
   )
  )
  Multiply(A=@n3, B=Constant3Vector() {constant=<rgba 55 55 0 1>})
  Multiply(A=@n4.RGB, B=VectorParameter() {parameterName=WeedsTone, defaultValue=<rgba 0.314 0.349 0.0363 0>}.RGB)
  ScalarParameter()
  n7 = ComponentMask(None=WorldPosition().XYZ) {g=true, r=true}
  n8 = Constant() {r=1024}
  n9 = LinearInterpolate(
   A=Constant() {r=0.2}
   B=Constant() {r=1.7}
   Alpha=TextureSample(UVs=Divide(A=@n7, B=@n8)) {samplerType=SAMPLERTYPE_Masks, texture=/Game/Athena/Environments/Island/Textures/T_FNBlendmask}.R
  )
  ShadingPathSwitch(Default=@n9, Deferred=@n9, Forward=@n9, Mobile=Constant() {r=0.5})
 ]
}
